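# Sound cue for the GrandmaC creature's spawn, layering a random spawn sound with a randomly chosen vocal howl.
# Each mixer plays the spawn layer quieter and the howl louder after a short fixed delay, then feeds a modulator.
# A quality-level switch sends its first two slots to the full wave pool and its last two to a smaller subset.
# Playback uses WolfMinion spawn attenuation, the tactical creature vocalization class and its own concurrency limit.
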SoundCue /Irwin/Sounds/GrandmaC/Cues/Cue_GrandmaC_Spawn
root=QualityLevel_0
nodes [
 {id=Delay_0, children [Random_0], properties {delayMax=0.1, delayMin=0.1}}
 {id=Delay_1, children [Random_3], properties {delayMax=0.1, delayMin=0.1}}
 {id=Mixer_0, children [Random_2,Delay_0], properties {inputVolume [0.4,1.3]}}
 {id=Mixer_2, children [Random_1,Delay_1], properties {inputVolume [0.4,1.3]}}
 {id=Modulator_0, children [Mixer_2]}
 {id=Modulator_1, children [Mixer_0]}
 {id=QualityLevel_0, children [Modulator_1,Modulator_1,Modulator_0,Modulator_0]}
 {
  id=Random_0
  children [WavePlayer_0,WavePlayer_1,WavePlayer_2,WavePlayer_3,WavePlayer_4,WavePlayer_6,WavePlayer_9]
  properties {weights [1,1,1,1,1,1,1]}
 }
 {id=Random_1, children [WavePlayer_10, WavePlayer_11], properties {weights [1,1]}}
 {id=Random_2, children [WavePlayer_5,WavePlayer_7,WavePlayer_8], properties {weights [1,1,1]}}
 {id=Random_3, children [WavePlayer_13, WavePlayer_14, WavePlayer_15], properties {weights [1,1,1]}}
 {id=WavePlayer_0, properties {soundWaveAssetPtr=/Irwin/Sounds/GrandmaC/Waves/GrandmaC_Vox_Howl_01}}
 {id=WavePlayer_1, properties {soundWaveAssetPtr=/Irwin/Sounds/GrandmaC/Waves/GrandmaC_Vox_Howl_02}}
 {id=WavePlayer_10, properties {soundWaveAssetPtr=/Irwin/Sounds/GrandmaC/Waves/GrandmaC_Spawn_01}}
 {id=WavePlayer_11, properties {soundWaveAssetPtr=/Irwin/Sounds/GrandmaC/Waves/GrandmaC_Spawn_02}}
 {id=WavePlayer_13, properties {soundWaveAssetPtr=&GrandmaC_Vox_Howl_01}}
 {id=WavePlayer_14, properties {soundWaveAssetPtr=&GrandmaC_Vox_Howl_02}}
 {id=WavePlayer_15, properties {soundWaveAssetPtr=/Irwin/Sounds/GrandmaC/Waves/GrandmaC_Vox_Howl_03}}
 {id=WavePlayer_2, properties {soundWaveAssetPtr=&GrandmaC_Vox_Howl_03}}
 {id=WavePlayer_3, properties {soundWaveAssetPtr=/Irwin/Sounds/GrandmaC/Waves/GrandmaC_Vox_Howl_04}}
 {id=WavePlayer_4, properties {soundWaveAssetPtr=/Irwin/Sounds/GrandmaC/Waves/GrandmaC_Vox_Howl_05}}
 {id=WavePlayer_5, properties {soundWaveAssetPtr=&GrandmaC_Spawn_01}}
 {id=WavePlayer_6, properties {soundWaveAssetPtr=/Irwin/Sounds/GrandmaC/Waves/GrandmaC_Vox_Howl_06}}
 {id=WavePlayer_7, properties {soundWaveAssetPtr=&GrandmaC_Spawn_02}}
 {id=WavePlayer_8, properties {soundWaveAssetPtr=/Irwin/Sounds/GrandmaC/Waves/GrandmaC_Spawn_03}}
 {id=WavePlayer_9, properties {soundWaveAssetPtr=/Irwin/Sounds/GrandmaC/Waves/GrandmaC_Vox_Howl_07}}
]
properties {
 attenuationSettings=/Irwin/Sounds/GrandmaC/Settings/Att_WolfMinion_Spawn
 bOverrideConcurrency=true
 bPrimeOnLoad=true
 maxDistance=12000
 soundClassObject=&Fort_Creatures_Vocalization_Tactical
 volumeMultiplier=1
 attenuationOverrides {falloffDistance=10000, stereoSpread=5000, attenuationShapeExtents {x=1000}}
 concurrencyOverrides {maxCount=4}
}
