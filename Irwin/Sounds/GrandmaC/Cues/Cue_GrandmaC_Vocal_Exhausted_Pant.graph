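# Sound cue for GrandmaC's exhausted panting vocalization, using a close-range prey vocalization attenuation and a tactical creature vocalization sound class.
# A Random node picks among six equally weighted pant waves and is excluded from branch culling.
# A Modulator at the root slightly randomizes volume and drops pitch noticeably below normal, with the whole cue played at reduced overall volume.
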SoundCue /Irwin/Sounds/GrandmaC/Cues/Cue_GrandmaC_Vocal_Exhausted_Pant
root=Modulator_0
nodes [
 {id=Modulator_0, children [Random_2], properties {pitchMax=0.85, pitchMin=0.8, volumeMin=0.75}}
 {
  id=Random_2
  children [WavePlayer_10, WavePlayer_11, WavePlayer_12, WavePlayer_13, WavePlayer_14, WavePlayer_15]
  properties {bShouldExcludeFromBranchCulling=true, weights [1,1,1,1,1,1]}
 }
 {id=WavePlayer_10, properties {soundWaveAssetPtr=/Irwin/Sounds/GrandmaC/Waves/GrandmaC_Vocal_Exhausted_Pant_01}}
 {id=WavePlayer_11, properties {soundWaveAssetPtr=/Irwin/Sounds/GrandmaC/Waves/GrandmaC_Vocal_Exhausted_Pant_02}}
 {id=WavePlayer_12, properties {soundWaveAssetPtr=/Irwin/Sounds/GrandmaC/Waves/GrandmaC_Vocal_Exhausted_Pant_03}}
 {id=WavePlayer_13, properties {soundWaveAssetPtr=/Irwin/Sounds/GrandmaC/Waves/GrandmaC_Vocal_Exhausted_Pant_04}}
 {id=WavePlayer_14, properties {soundWaveAssetPtr=/Irwin/Sounds/GrandmaC/Waves/GrandmaC_Vocal_Exhausted_Pant_05}}
 {id=WavePlayer_15, properties {soundWaveAssetPtr=/Irwin/Sounds/GrandmaC/Waves/GrandmaC_Vocal_Exhausted_Pant_06}}
]
properties {
 attenuationSettings=/Irwin/Sounds/Settings/Prey_Vocalization_Close_Att
 maxDistance=2150
 soundClassObject=&Fort_Creatures_Vocalization_Tactical
 volumeMultiplier=0.5
}
