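# Sound cue for GrandmaC creature sprint vocals, with player-riding variants that respond to fatigue.
# IsLocal and IsRiddenByEnemy branches choose sound class, attenuation and volume, all converging on one shared random-volume modulator.
# A quality-level node then picks a fuller or smaller-pool FatigueLevel crossfade over randomized breaths, heavy breathing and tiring whines.
# IsEnemy swaps sprint breaths for aggressive breaths, and empty random slots leave some plays silent.
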SoundCue /Irwin/Sounds/GrandmaC/Cues/Cue_GrandmaC_Vocal_PlayerSprint
root=Branch_4
nodes [
 {id=Attenuation_0, children [Modulator_2], properties {attenuationSettings=/Irwin/Sounds/Settings/Predator_Vocalization_Att}}
 {id=Attenuation_5, children [Modulator_2], properties {attenuationSettings=/Game/Sounds/Attenuation/Fort_Riding_Enemy_VO_Att}}
 {id=Attenuation_6, children [Modulator_2], properties {attenuationSettings=&Predator_Vocalization_Att}}
 {id=Attenuation_7, children [Modulator_2], properties {attenuationSettings=&Predator_Vocalization_Att}}
 {id=Branch_1, children [Random_0,Random_1,Random_1], properties {boolParameterName=IsEnemy}}
 {id=Branch_2, children [Random_5,Random_6,Random_6], properties {boolParameterName=IsEnemy}}
 {id=Branch_3, children [Modulator_1,Modulator_4,SoundClass_0], properties {boolParameterName=IsRiddenByEnemy}}
 {id=Branch_4, children [SoundClass_7,Branch_3,Branch_3], properties {boolParameterName=IsLocal}}
 {id=Modulator_1, children [SoundClass_5], properties {pitchMax=1, pitchMin=1, volumeMax=0.7, volumeMin=0.7}}
 {id=Modulator_2, children [QualityLevel_0], properties {volumeMax=1, volumeMin=0.5}}
 {id=Modulator_4, children [SoundClass_6], properties {pitchMax=1, pitchMin=1, volumeMax=0.1, volumeMin=0.1}}
 {
  id=ParamCrossFade_0
  children [Branch_2,Random_4,Random_3]
  properties {
   paramName=FatigueLevel
   crossFadeInput [
    {fadeInDistanceEnd=0, fadeInDistanceStart=0, fadeOutDistanceEnd=1, fadeOutDistanceStart=0.9, volume=1.8}
    {fadeInDistanceEnd=0.92, fadeInDistanceStart=0.88, fadeOutDistanceEnd=1, fadeOutDistanceStart=1, volume=2.5}
    {fadeInDistanceEnd=0, fadeInDistanceStart=0, fadeOutDistanceEnd=0, fadeOutDistanceStart=0, volume=2.2}
   ]
  }
 }
 {
  id=ParamCrossFade_2
  children [Branch_1,Random_2,Random_11]
  properties {
   paramName=FatigueLevel
   crossFadeInput [
    {fadeInDistanceEnd=0, fadeInDistanceStart=0, fadeOutDistanceEnd=0.92, fadeOutDistanceStart=0.88, volume=1.8}
    {fadeInDistanceEnd=0.92, fadeInDistanceStart=0.88, fadeOutDistanceEnd=0.96, fadeOutDistanceStart=0.95, volume=2.5}
    {fadeInDistanceEnd=0.96, fadeInDistanceStart=0.95, fadeOutDistanceEnd=1, fadeOutDistanceStart=1, volume=2.2}
   ]
  }
 }
 {id=QualityLevel_0, children [ParamCrossFade_2, ParamCrossFade_2, ParamCrossFade_0, ParamCrossFade_0]}
 {
  id=Random_0
  children [
   WavePlayer_12
   WavePlayer_28
   WavePlayer_13
   WavePlayer_14
   WavePlayer_32
   WavePlayer_20
   WavePlayer_36
   WavePlayer_21
   WavePlayer_37
   WavePlayer_22
   WavePlayer_38
   WavePlayer_23
   WavePlayer_47
   WavePlayer_24
   WavePlayer_48
   WavePlayer_25
   WavePlayer_51
   WavePlayer_26
   WavePlayer_54
   WavePlayer_27
   WavePlayer_59
   null
   null
   null
   null
   null
   null
   null
   null
  ]
  properties {bShouldExcludeFromBranchCulling=true, weights [1,1,1,1,1,1,1,1,1,1,1,1,1,1,1,1,1,1,1,1,1,1,1,1,1,1,1,1,1]}
 }
 {
  id=Random_1
  children [
   WavePlayer_60
   WavePlayer_61
   WavePlayer_62
   WavePlayer_63
   WavePlayer_64
   WavePlayer_66
   WavePlayer_68
   WavePlayer_70
   WavePlayer_71
   WavePlayer_72
   WavePlayer_74
   WavePlayer_75
  ]
  properties {bShouldExcludeFromBranchCulling=true, weights [1,1,1,1,1,1,1,1,1,1,1,1]}
 }
 {
  id=Random_11
  children [WavePlayer_96, WavePlayer_108, WavePlayer_109, WavePlayer_110, WavePlayer_111, WavePlayer_112, WavePlayer_113, WavePlayer_114, null, null]
  properties {bShouldExcludeFromBranchCulling=true, weights [1,1,1,1,1,1,1,1,1,1]}
 }
 {
  id=Random_2
  children [
   WavePlayer_76
   WavePlayer_77
   WavePlayer_78
   WavePlayer_79
   WavePlayer_90
   WavePlayer_91
   WavePlayer_92
   WavePlayer_93
   null
   null
   null
   null
   null
  ]
  properties {bShouldExcludeFromBranchCulling=true, weights [1,1,1,1,1,1,1,1,1,1,1,1,1]}
 }
 {
  id=Random_3
  children [WavePlayer_127, WavePlayer_128, WavePlayer_129, WavePlayer_130, WavePlayer_131, WavePlayer_132, WavePlayer_133, null, null]
  properties {bShouldExcludeFromBranchCulling=true, weights [1,1,1,1,1,1,1,1,1]}
 }
 {
  id=Random_4
  children [WavePlayer_123, WavePlayer_124, WavePlayer_125, WavePlayer_126, null, null, null, null, null, null]
  properties {bShouldExcludeFromBranchCulling=true, weights [1,1,1,1,1,1,1,1,1,1]}
 }
 {
  id=Random_5
  children [WavePlayer_115, WavePlayer_117, WavePlayer_116, null, WavePlayer_118, null, null, null]
  properties {bShouldExcludeFromBranchCulling=true, weights [1,1,1,1,1,1,1,1]}
 }
 {
  id=Random_6
  children [WavePlayer_119, WavePlayer_120, WavePlayer_121, WavePlayer_122, null, null, null, null]
  properties {bShouldExcludeFromBranchCulling=true, weights [1,1,1,1,1,1,1,1]}
 }
 {
  id=SoundClass_0
  children [Attenuation_0]
  properties {soundClassOverride=&Fort_Creatures_Vocalization_Tactical}
 }
 {
  id=SoundClass_5
  children [Attenuation_5]
  properties {soundClassOverride=&Fort_Player_Foley_3P_Enemy_Athena}
 }
 {
  id=SoundClass_6
  children [Attenuation_6]
  properties {soundClassOverride=&Fort_Creatures_Vocalization_Tactical}
 }
 {
  id=SoundClass_7
  children [Attenuation_7]
  properties {soundClassOverride=&Fort_Creatures_Vocalization_Tactical}
 }
 {id=WavePlayer_108, properties {soundWaveAssetPtr=/Irwin/Sounds/GrandmaC/Waves/GrandmaC_Vocal_Sprint_Tiring_Whines_02}}
 {id=WavePlayer_109, properties {soundWaveAssetPtr=/Irwin/Sounds/GrandmaC/Waves/GrandmaC_Vocal_Sprint_Tiring_Whines_03}}
 {id=WavePlayer_110, properties {soundWaveAssetPtr=/Irwin/Sounds/GrandmaC/Waves/GrandmaC_Vocal_Sprint_Tiring_Whines_04}}
 {id=WavePlayer_111, properties {soundWaveAssetPtr=/Irwin/Sounds/GrandmaC/Waves/GrandmaC_Vox_Swim_Aggro_Breath_01}}
 {id=WavePlayer_112, properties {soundWaveAssetPtr=/Irwin/Sounds/GrandmaC/Waves/GrandmaC_Vox_Swim_Aggro_Breath_02}}
 {id=WavePlayer_113, properties {soundWaveAssetPtr=/Irwin/Sounds/GrandmaC/Waves/GrandmaC_Vox_Swim_Aggro_Breath_03}}
 {id=WavePlayer_114, properties {soundWaveAssetPtr=/Irwin/Sounds/GrandmaC/Waves/GrandmaC_Vox_Swim_Aggro_Breath_04}}
 {id=WavePlayer_115, properties {soundWaveAssetPtr=&GrandmaC_Vox_Swim_Aggro_Breath_01}}
 {id=WavePlayer_116, properties {soundWaveAssetPtr=&GrandmaC_Vox_Swim_Aggro_Breath_03}}
 {id=WavePlayer_117, properties {soundWaveAssetPtr=&GrandmaC_Vox_Swim_Aggro_Breath_02}}
 {id=WavePlayer_118, properties {soundWaveAssetPtr=&GrandmaC_Vox_Swim_Aggro_Breath_04}}
 {id=WavePlayer_119, properties {soundWaveAssetPtr=/Irwin/Sounds/GrandmaC/Waves/GrandmaC_Vox_Sprint_Breath_01}}
 {id=WavePlayer_12, properties {soundWaveAssetPtr=&GrandmaC_Vox_Swim_Aggro_Breath_01}}
 {id=WavePlayer_120, properties {soundWaveAssetPtr=/Irwin/Sounds/GrandmaC/Waves/GrandmaC_Vox_Sprint_Breath_02}}
 {id=WavePlayer_121, properties {soundWaveAssetPtr=/Irwin/Sounds/GrandmaC/Waves/GrandmaC_Vox_Sprint_Breath_03}}
 {id=WavePlayer_122, properties {soundWaveAssetPtr=/Irwin/Sounds/GrandmaC/Waves/GrandmaC_Vox_Sprint_Breath_04}}
 {id=WavePlayer_123, properties {soundWaveAssetPtr=/Irwin/Sounds/GrandmaC/Waves/GrandmaC_Vocal_Heavy_Breathing_01}}
 {id=WavePlayer_124, properties {soundWaveAssetPtr=/Irwin/Sounds/GrandmaC/Waves/GrandmaC_Vocal_Heavy_Breathing_02}}
 {id=WavePlayer_125, properties {soundWaveAssetPtr=/Irwin/Sounds/GrandmaC/Waves/GrandmaC_Vocal_Heavy_Breathing_03}}
 {id=WavePlayer_126, properties {soundWaveAssetPtr=/Irwin/Sounds/GrandmaC/Waves/GrandmaC_Vocal_Heavy_Breathing_04}}
 {id=WavePlayer_127, properties {soundWaveAssetPtr=/Irwin/Sounds/GrandmaC/Waves/GrandmaC_Vocal_Sprint_Tiring_Whines_01}}
 {id=WavePlayer_128, properties {soundWaveAssetPtr=&GrandmaC_Vocal_Sprint_Tiring_Whines_02}}
 {id=WavePlayer_129, properties {soundWaveAssetPtr=&GrandmaC_Vocal_Sprint_Tiring_Whines_03}}
 {id=WavePlayer_13, properties {soundWaveAssetPtr=&GrandmaC_Vox_Swim_Aggro_Breath_02}}
 {id=WavePlayer_130, properties {soundWaveAssetPtr=&GrandmaC_Vocal_Sprint_Tiring_Whines_04}}
 {id=WavePlayer_131, properties {soundWaveAssetPtr=/Irwin/Sounds/GrandmaC/Waves/GrandmaC_Vox_Swim_Aggro_Breath_11}}
 {id=WavePlayer_132, properties {soundWaveAssetPtr=/Irwin/Sounds/GrandmaC/Waves/GrandmaC_Vox_Swim_Aggro_Breath_12}}
 {id=WavePlayer_133, properties {soundWaveAssetPtr=/Irwin/Sounds/GrandmaC/Waves/GrandmaC_Vox_Swim_Aggro_Breath_13}}
 {id=WavePlayer_14, properties {soundWaveAssetPtr=&GrandmaC_Vox_Swim_Aggro_Breath_03}}
 {id=WavePlayer_20, properties {soundWaveAssetPtr=&GrandmaC_Vox_Swim_Aggro_Breath_04}}
 {id=WavePlayer_21, properties {soundWaveAssetPtr=/Irwin/Sounds/GrandmaC/Waves/GrandmaC_Vox_Swim_Aggro_Breath_05}}
 {id=WavePlayer_22, properties {soundWaveAssetPtr=/Irwin/Sounds/GrandmaC/Waves/GrandmaC_Vox_Swim_Aggro_Breath_06}}
 {id=WavePlayer_23, properties {soundWaveAssetPtr=/Irwin/Sounds/GrandmaC/Waves/GrandmaC_Vox_Swim_Aggro_Breath_07}}
 {id=WavePlayer_24, properties {soundWaveAssetPtr=/Irwin/Sounds/GrandmaC/Waves/GrandmaC_Vox_Swim_Aggro_Breath_08}}
 {id=WavePlayer_25, properties {soundWaveAssetPtr=/Irwin/Sounds/GrandmaC/Waves/GrandmaC_Vox_Swim_Aggro_Breath_09}}
 {id=WavePlayer_26, properties {soundWaveAssetPtr=/Irwin/Sounds/GrandmaC/Waves/GrandmaC_Vox_Swim_Aggro_Breath_10}}
 {id=WavePlayer_27, properties {soundWaveAssetPtr=&GrandmaC_Vox_Swim_Aggro_Breath_11}}
 {id=WavePlayer_28, properties {soundWaveAssetPtr=&GrandmaC_Vox_Swim_Aggro_Breath_12}}
 {id=WavePlayer_32, properties {soundWaveAssetPtr=&GrandmaC_Vox_Swim_Aggro_Breath_13}}
 {id=WavePlayer_36, properties {soundWaveAssetPtr=/Irwin/Sounds/GrandmaC/Waves/GrandmaC_Vox_Swim_Aggro_Breath_14}}
 {id=WavePlayer_37, properties {soundWaveAssetPtr=/Irwin/Sounds/GrandmaC/Waves/GrandmaC_Vox_Swim_Aggro_Breath_15}}
 {id=WavePlayer_38, properties {soundWaveAssetPtr=/Irwin/Sounds/GrandmaC/Waves/GrandmaC_Vox_Swim_Aggro_Breath_16}}
 {id=WavePlayer_47, properties {soundWaveAssetPtr=/Irwin/Sounds/GrandmaC/Waves/GrandmaC_Vox_Swim_Aggro_Breath_17}}
 {id=WavePlayer_48, properties {soundWaveAssetPtr=/Irwin/Sounds/GrandmaC/Waves/GrandmaC_Vox_Swim_Aggro_Breath_18}}
 {id=WavePlayer_51, properties {soundWaveAssetPtr=/Irwin/Sounds/GrandmaC/Waves/GrandmaC_Vox_Swim_Aggro_Breath_19}}
 {id=WavePlayer_54, properties {soundWaveAssetPtr=/Irwin/Sounds/GrandmaC/Waves/GrandmaC_Vox_Swim_Aggro_Breath_20}}
 {id=WavePlayer_59, properties {soundWaveAssetPtr=/Irwin/Sounds/GrandmaC/Waves/GrandmaC_Vox_Swim_Aggro_Breath_21}}
 {id=WavePlayer_60, properties {soundWaveAssetPtr=/Irwin/Sounds/GrandmaC/Waves/GrandmaC_Vox_Sprint_Breath_12}}
 {id=WavePlayer_61, properties {soundWaveAssetPtr=/Irwin/Sounds/GrandmaC/Waves/GrandmaC_Vox_Sprint_Breath_11}}
 {id=WavePlayer_62, properties {soundWaveAssetPtr=/Irwin/Sounds/GrandmaC/Waves/GrandmaC_Vox_Sprint_Breath_10}}
 {id=WavePlayer_63, properties {soundWaveAssetPtr=/Irwin/Sounds/GrandmaC/Waves/GrandmaC_Vox_Sprint_Breath_09}}
 {id=WavePlayer_64, properties {soundWaveAssetPtr=/Irwin/Sounds/GrandmaC/Waves/GrandmaC_Vox_Sprint_Breath_08}}
 {id=WavePlayer_66, properties {soundWaveAssetPtr=/Irwin/Sounds/GrandmaC/Waves/GrandmaC_Vox_Sprint_Breath_07}}
 {id=WavePlayer_68, properties {soundWaveAssetPtr=/Irwin/Sounds/GrandmaC/Waves/GrandmaC_Vox_Sprint_Breath_06}}
 {id=WavePlayer_70, properties {soundWaveAssetPtr=/Irwin/Sounds/GrandmaC/Waves/GrandmaC_Vox_Sprint_Breath_05}}
 {id=WavePlayer_71, properties {soundWaveAssetPtr=&GrandmaC_Vox_Sprint_Breath_04}}
 {id=WavePlayer_72, properties {soundWaveAssetPtr=&GrandmaC_Vox_Sprint_Breath_03}}
 {id=WavePlayer_74, properties {soundWaveAssetPtr=&GrandmaC_Vox_Sprint_Breath_02}}
 {id=WavePlayer_75, properties {soundWaveAssetPtr=&GrandmaC_Vox_Sprint_Breath_01}}
 {id=WavePlayer_76, properties {soundWaveAssetPtr=&GrandmaC_Vocal_Heavy_Breathing_01}}
 {id=WavePlayer_77, properties {soundWaveAssetPtr=&GrandmaC_Vocal_Heavy_Breathing_02}}
 {id=WavePlayer_78, properties {soundWaveAssetPtr=&GrandmaC_Vocal_Heavy_Breathing_03}}
 {id=WavePlayer_79, properties {soundWaveAssetPtr=&GrandmaC_Vocal_Heavy_Breathing_04}}
 {id=WavePlayer_90, properties {soundWaveAssetPtr=/Irwin/Sounds/GrandmaC/Waves/GrandmaC_Vocal_Heavy_Breathing_05}}
 {id=WavePlayer_91, properties {soundWaveAssetPtr=/Irwin/Sounds/GrandmaC/Waves/GrandmaC_Vocal_Heavy_Breathing_06}}
 {id=WavePlayer_92, properties {soundWaveAssetPtr=/Irwin/Sounds/GrandmaC/Waves/GrandmaC_Vocal_Heavy_Breathing_07}}
 {id=WavePlayer_93, properties {soundWaveAssetPtr=/Irwin/Sounds/GrandmaC/Waves/GrandmaC_Vocal_Heavy_Breathing_08}}
 {id=WavePlayer_96, properties {soundWaveAssetPtr=&GrandmaC_Vocal_Sprint_Tiring_Whines_01}}
]
properties {
 maxDistance=9000
 pitchMultiplier=0.9
 volumeMultiplier=1.3
 concurrencySet [
  /Game/Sounds/Concurrency_Rules/Burt_Vocal_Conc
  &Athena_Footstep_Global_Conc
  /Game/Sounds/Concurrency_Rules/LimitToOnePreventNewQuick
 ]
}
