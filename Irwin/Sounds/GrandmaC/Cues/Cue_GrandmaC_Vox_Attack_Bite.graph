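# Sound cue for GrandmaC's bite attack vocalization, played as a creature vocalization with wolf-minion attenuation.
# A Modulator feeds a quality switch that sends the first two levels to a random pick of eleven bite waves and the last two to a pick of only the first five.
# The larger random node is excluded from branch culling.
# Volume is boosted, and concurrency borrows the Athena local and global footstep limits.
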SoundCue /Irwin/Sounds/GrandmaC/Cues/Cue_GrandmaC_Vox_Attack_Bite
root=Modulator_1
nodes [
 {id=Modulator_1, children [QualityLevel_0]}
 {id=QualityLevel_0, children [Random_1,Random_1,Random_3,Random_3]}
 {
  id=Random_1
  children [
   WavePlayer_14
   WavePlayer_15
   WavePlayer_16
   WavePlayer_17
   WavePlayer_18
   WavePlayer_19
   WavePlayer_20
   WavePlayer_21
   WavePlayer_22
   WavePlayer_23
   WavePlayer_24
  ]
  properties {bShouldExcludeFromBranchCulling=true, weights [1,1,1,1,1,1,1,1,1,1,1]}
 }
 {id=Random_3, children [WavePlayer_25, WavePlayer_26, WavePlayer_27, WavePlayer_30, WavePlayer_31], properties {weights [1,1,1,1,1]}}
 {id=WavePlayer_14, properties {soundWaveAssetPtr=/Irwin/Sounds/GrandmaC/Waves/GrandmaC_Vox_Attack_Bite_01}}
 {id=WavePlayer_15, properties {soundWaveAssetPtr=/Irwin/Sounds/GrandmaC/Waves/GrandmaC_Vox_Attack_Bite_02}}
 {id=WavePlayer_16, properties {soundWaveAssetPtr=/Irwin/Sounds/GrandmaC/Waves/GrandmaC_Vox_Attack_Bite_03}}
 {id=WavePlayer_17, properties {soundWaveAssetPtr=/Irwin/Sounds/GrandmaC/Waves/GrandmaC_Vox_Attack_Bite_04}}
 {id=WavePlayer_18, properties {soundWaveAssetPtr=/Irwin/Sounds/GrandmaC/Waves/GrandmaC_Vox_Attack_Bite_05}}
 {id=WavePlayer_19, properties {soundWaveAssetPtr=/Irwin/Sounds/GrandmaC/Waves/GrandmaC_Vox_Attack_Bite_06}}
 {id=WavePlayer_20, properties {soundWaveAssetPtr=/Irwin/Sounds/GrandmaC/Waves/GrandmaC_Vox_Attack_Bite_07}}
 {id=WavePlayer_21, properties {soundWaveAssetPtr=/Irwin/Sounds/GrandmaC/Waves/GrandmaC_Vox_Attack_Bite_08}}
 {id=WavePlayer_22, properties {soundWaveAssetPtr=/Irwin/Sounds/GrandmaC/Waves/GrandmaC_Vox_Attack_Bite_09}}
 {id=WavePlayer_23, properties {soundWaveAssetPtr=/Irwin/Sounds/GrandmaC/Waves/GrandmaC_Vox_Attack_Bite_10}}
 {id=WavePlayer_24, properties {soundWaveAssetPtr=/Irwin/Sounds/GrandmaC/Waves/GrandmaC_Vox_Attack_Bite_11}}
 {id=WavePlayer_25, properties {soundWaveAssetPtr=&GrandmaC_Vox_Attack_Bite_01}}
 {id=WavePlayer_26, properties {soundWaveAssetPtr=&GrandmaC_Vox_Attack_Bite_02}}
 {id=WavePlayer_27, properties {soundWaveAssetPtr=&GrandmaC_Vox_Attack_Bite_03}}
 {id=WavePlayer_30, properties {soundWaveAssetPtr=&GrandmaC_Vox_Attack_Bite_04}}
 {id=WavePlayer_31, properties {soundWaveAssetPtr=&GrandmaC_Vox_Attack_Bite_05}}
]
properties {
 attenuationSettings=/Irwin/Sounds/GrandmaC/Settings/Att_WolfMinion_Vocalization
 maxDistance=10000
 soundClassObject=&Fort_Creatures_Vocalization_Tactical
 volumeMultiplier=2.5
 concurrencySet [&Athena_Footstep_Local_Conc, &Athena_Footstep_Global_Conc]
}
