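# Sound cue for the GrandmaC creature's vocalization during a lunge attack.
# A Random node picks among six equally weighted lunge wave variations and feeds a Modulator at the root.
# Playback uses WolfMinion vocalization attenuation and the creature tactical vocalization sound class, with a boosted volume.
# Its concurrency limits come from the Athena local and global footstep sets rather than vocal-specific ones.
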SoundCue /Irwin/Sounds/GrandmaC/Cues/Cue_GrandmaC_Vox_Attack_Lunge
root=Modulator_1
nodes [
 {id=Modulator_1, children [Random_1]}
 {
  id=Random_1
  children [WavePlayer_6, WavePlayer_7, WavePlayer_8, WavePlayer_9, WavePlayer_10, WavePlayer_11]
  properties {bShouldExcludeFromBranchCulling=true, weights [1,1,1,1,1,1]}
 }
 {id=WavePlayer_10, properties {soundWaveAssetPtr=/Irwin/Sounds/GrandmaC/Waves/GrandmaC_Vox_Attack_Lunge_02}}
 {id=WavePlayer_11, properties {soundWaveAssetPtr=/Irwin/Sounds/GrandmaC/Waves/GrandmaC_Vox_Attack_Lunge_01}}
 {id=WavePlayer_6, properties {soundWaveAssetPtr=/Irwin/Sounds/GrandmaC/Waves/GrandmaC_Vox_Attack_Lunge_06}}
 {id=WavePlayer_7, properties {soundWaveAssetPtr=/Irwin/Sounds/GrandmaC/Waves/GrandmaC_Vox_Attack_Lunge_05}}
 {id=WavePlayer_8, properties {soundWaveAssetPtr=/Irwin/Sounds/GrandmaC/Waves/GrandmaC_Vox_Attack_Lunge_04}}
 {id=WavePlayer_9, properties {soundWaveAssetPtr=/Irwin/Sounds/GrandmaC/Waves/GrandmaC_Vox_Attack_Lunge_03}}
]
properties {
 attenuationSettings=/Irwin/Sounds/GrandmaC/Settings/Att_WolfMinion_Vocalization
 maxDistance=10000
 soundClassObject=&Fort_Creatures_Vocalization_Tactical
 volumeMultiplier=2.5
 concurrencySet [&Athena_Footstep_Local_Conc, &Athena_Footstep_Global_Conc]
}
