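# Sound cue for GrandmaC sprint breathing vocals, using shared WolfMinion vocalization attenuation and footstep and per-owner concurrency limits.
# Branches on IsLocal, then IsRiddenByEnemy for non-local, set sound class: enemy-ridden uses enemy player foley at full volume, not enemy-ridden is heavily attenuated.
# All paths share one randomized volume modulator into a quality-level switch between a 19-wave random pool and a four-wave subset, each pitch and volume modulated.
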SoundCue /Irwin/Sounds/GrandmaC/Cues/Cue_GrandmaC_Vox_Breath_Sprint
root=Branch_4
nodes [
 {id=Attenuation_0, children [Modulator_3], properties {attenuationSettings=/Irwin/Sounds/GrandmaC/Settings/Att_WolfMinion_Vocalization}}
 {id=Attenuation_5, children [Modulator_3], properties {attenuationSettings=&Att_WolfMinion_Vocalization}}
 {id=Attenuation_6, children [Modulator_3], properties {attenuationSettings=&Att_WolfMinion_Vocalization}}
 {id=Attenuation_7, children [Modulator_3], properties {attenuationSettings=&Att_WolfMinion_Vocalization}}
 {id=Branch_3, children [Modulator_5,Modulator_4,SoundClass_0], properties {boolParameterName=IsRiddenByEnemy}}
 {id=Branch_4, children [SoundClass_7,Branch_3,Branch_3], properties {boolParameterName=IsLocal}}
 {id=Modulator_0, children [Random_0], properties {pitchMin=1, volumeMin=0.5}}
 {id=Modulator_1, children [Random_1], properties {pitchMax=1.1, pitchMin=1, volumeMin=0.5}}
 {id=Modulator_3, children [QualityLevel_0], properties {volumeMax=1, volumeMin=0.5}}
 {id=Modulator_4, children [SoundClass_6], properties {pitchMax=1, pitchMin=1, volumeMax=0.1, volumeMin=0.1}}
 {id=Modulator_5, children [SoundClass_5], properties {pitchMax=1, pitchMin=1, volumeMax=1, volumeMin=1}}
 {id=QualityLevel_0, children [Modulator_0,Modulator_0,Modulator_1,Modulator_1]}
 {
  id=Random_0
  children [
   WavePlayer_0
   WavePlayer_1
   WavePlayer_2
   WavePlayer_3
   WavePlayer_4
   WavePlayer_5
   WavePlayer_6
   WavePlayer_7
   WavePlayer_8
   WavePlayer_9
   WavePlayer_10
   WavePlayer_11
   WavePlayer_12
   WavePlayer_13
   WavePlayer_14
   WavePlayer_15
   WavePlayer_16
   WavePlayer_17
   WavePlayer_41
  ]
  properties {bShouldExcludeFromBranchCulling=true, weights [1,1,1,1,1,1,1,1,1,1,1,1,1,1,1,1,1,1,1]}
 }
 {id=Random_1, children [WavePlayer_42, WavePlayer_43, WavePlayer_44, WavePlayer_45], properties {weights [1,1,1,1]}}
 {
  id=SoundClass_0
  children [Attenuation_0]
  properties {soundClassOverride=&Fort_Creatures_Vocalization_Tactical}
 }
 {
  id=SoundClass_5
  children [Attenuation_5]
  properties {soundClassOverride=&Fort_Player_Foley_3P_Enemy_Athena}
 }
 {
  id=SoundClass_6
  children [Attenuation_6]
  properties {soundClassOverride=&Fort_Creatures_Vocalization_Tactical}
 }
 {
  id=SoundClass_7
  children [Attenuation_7]
  properties {soundClassOverride=&Fort_Creatures_Vocalization_Tactical}
 }
 {id=WavePlayer_0, properties {soundWaveAssetPtr=/Irwin/Sounds/GrandmaC/Waves/GrandmaC_Vox_Sprint_Breath_01}}
 {id=WavePlayer_1, properties {soundWaveAssetPtr=/Irwin/Sounds/GrandmaC/Waves/GrandmaC_Vox_Sprint_Breath_02}}
 {id=WavePlayer_10, properties {soundWaveAssetPtr=/Irwin/Sounds/GrandmaC/Waves/GrandmaC_Vox_Sprint_Breath_11}}
 {id=WavePlayer_11, properties {soundWaveAssetPtr=/Irwin/Sounds/GrandmaC/Waves/GrandmaC_Vox_Sprint_Breath_12}}
 {id=WavePlayer_12, properties {soundWaveAssetPtr=/Irwin/Sounds/GrandmaC/Waves/GrandmaC_Vox_Sprint_Breath_13}}
 {id=WavePlayer_13, properties {soundWaveAssetPtr=/Irwin/Sounds/GrandmaC/Waves/GrandmaC_Vox_Sprint_Breath_14}}
 {id=WavePlayer_14, properties {soundWaveAssetPtr=/Irwin/Sounds/GrandmaC/Waves/GrandmaC_Vox_Sprint_Breath_15}}
 {id=WavePlayer_15, properties {soundWaveAssetPtr=/Irwin/Sounds/GrandmaC/Waves/GrandmaC_Vox_Sprint_Breath_16}}
 {id=WavePlayer_16, properties {soundWaveAssetPtr=/Irwin/Sounds/GrandmaC/Waves/GrandmaC_Vox_Sprint_Breath_17}}
 {id=WavePlayer_17, properties {soundWaveAssetPtr=/Irwin/Sounds/GrandmaC/Waves/GrandmaC_Vox_Sprint_Breath_18}}
 {id=WavePlayer_2, properties {soundWaveAssetPtr=/Irwin/Sounds/GrandmaC/Waves/GrandmaC_Vox_Sprint_Breath_03}}
 {id=WavePlayer_3, properties {soundWaveAssetPtr=/Irwin/Sounds/GrandmaC/Waves/GrandmaC_Vox_Sprint_Breath_04}}
 {id=WavePlayer_4, properties {soundWaveAssetPtr=/Irwin/Sounds/GrandmaC/Waves/GrandmaC_Vox_Sprint_Breath_05}}
 {id=WavePlayer_41, properties {soundWaveAssetPtr=/Irwin/Sounds/GrandmaC/Waves/GrandmaC_Vox_Sprint_Breath_19}}
 {id=WavePlayer_42, properties {soundWaveAssetPtr=&GrandmaC_Vox_Sprint_Breath_01}}
 {id=WavePlayer_43, properties {soundWaveAssetPtr=&GrandmaC_Vox_Sprint_Breath_02}}
 {id=WavePlayer_44, properties {soundWaveAssetPtr=&GrandmaC_Vox_Sprint_Breath_03}}
 {id=WavePlayer_45, properties {soundWaveAssetPtr=&GrandmaC_Vox_Sprint_Breath_04}}
 {id=WavePlayer_5, properties {soundWaveAssetPtr=/Irwin/Sounds/GrandmaC/Waves/GrandmaC_Vox_Sprint_Breath_06}}
 {id=WavePlayer_6, properties {soundWaveAssetPtr=/Irwin/Sounds/GrandmaC/Waves/GrandmaC_Vox_Sprint_Breath_07}}
 {id=WavePlayer_7, properties {soundWaveAssetPtr=/Irwin/Sounds/GrandmaC/Waves/GrandmaC_Vox_Sprint_Breath_08}}
 {id=WavePlayer_8, properties {soundWaveAssetPtr=/Irwin/Sounds/GrandmaC/Waves/GrandmaC_Vox_Sprint_Breath_09}}
 {id=WavePlayer_9, properties {soundWaveAssetPtr=/Irwin/Sounds/GrandmaC/Waves/GrandmaC_Vox_Sprint_Breath_10}}
]
properties {
 maxDistance=10000
 volumeMultiplier=1.7
 concurrencySet [
  &Athena_Footstep_Local_Conc
  &Athena_Footstep_Global_Conc
  /Irwin/Sounds/Settings/Conc/SetOneLimitToOwner_VO_Conc
 ]
}
