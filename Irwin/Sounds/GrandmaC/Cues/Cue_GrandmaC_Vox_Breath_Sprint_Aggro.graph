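# Sound cue for GrandmaC's aggressive sprint breathing vocal, built from swim aggro breath waves and limited by footstep concurrency sets.
# IsLocal, then IsRiddenByEnemy, selects the sound class: ridden by an enemy uses enemy foley, explicitly not ridden plays much quieter, other paths use tactical creature vocalization.
# All paths share WolfMinion attenuation and randomized modulation.
# A quality switch picks either a 21-wave random pool or a briefly delayed four-wave pool.
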SoundCue /Irwin/Sounds/GrandmaC/Cues/Cue_GrandmaC_Vox_Breath_Sprint_Aggro
root=Branch_4
nodes [
 {id=Attenuation_0, children [Modulator_2], properties {attenuationSettings=/Irwin/Sounds/GrandmaC/Settings/Att_WolfMinion_Vocalization}}
 {id=Attenuation_5, children [Modulator_2], properties {attenuationSettings=&Att_WolfMinion_Vocalization}}
 {id=Attenuation_6, children [Modulator_2], properties {attenuationSettings=&Att_WolfMinion_Vocalization}}
 {id=Attenuation_7, children [Modulator_2], properties {attenuationSettings=&Att_WolfMinion_Vocalization}}
 {id=Branch_3, children [Modulator_3,Modulator_4,SoundClass_0], properties {boolParameterName=IsRiddenByEnemy}}
 {id=Branch_4, children [SoundClass_7,Branch_3,Branch_3], properties {boolParameterName=IsLocal}}
 {id=Delay_1, children [Modulator_1], properties {delayMax=0.2}}
 {id=Modulator_0, children [Random_1], properties {pitchMin=1, volumeMin=0.5}}
 {id=Modulator_1, children [Random_0], properties {pitchMax=1.1, volumeMin=0.5}}
 {id=Modulator_2, children [QualityLevel_0], properties {volumeMax=1, volumeMin=0.5}}
 {id=Modulator_3, children [SoundClass_5], properties {pitchMax=1, pitchMin=1, volumeMax=1, volumeMin=1}}
 {id=Modulator_4, children [SoundClass_6], properties {pitchMax=1, pitchMin=1, volumeMax=0.1, volumeMin=0.1}}
 {id=QualityLevel_0, children [Modulator_0,Modulator_0,Delay_1,Delay_1]}
 {
  id=Random_0
  children [WavePlayer_31, WavePlayer_47, WavePlayer_48, WavePlayer_49]
  properties {bShouldExcludeFromBranchCulling=true, weights [1,1,1,1]}
 }
 {
  id=Random_1
  children [
   WavePlayer_0
   WavePlayer_1
   WavePlayer_2
   WavePlayer_3
   WavePlayer_14
   WavePlayer_15
   WavePlayer_16
   WavePlayer_17
   WavePlayer_18
   WavePlayer_19
   WavePlayer_20
   WavePlayer_21
   WavePlayer_22
   WavePlayer_23
   WavePlayer_24
   WavePlayer_25
   WavePlayer_26
   WavePlayer_27
   WavePlayer_28
   WavePlayer_29
   WavePlayer_30
  ]
  properties {bShouldExcludeFromBranchCulling=true, weights [1,1,1,1,1,1,1,1,1,1,1,1,1,1,1,1,1,1,1,1,1]}
 }
 {
  id=SoundClass_0
  children [Attenuation_0]
  properties {soundClassOverride=&Fort_Creatures_Vocalization_Tactical}
 }
 {
  id=SoundClass_5
  children [Attenuation_5]
  properties {soundClassOverride=&Fort_Player_Foley_3P_Enemy_Athena}
 }
 {
  id=SoundClass_6
  children [Attenuation_6]
  properties {soundClassOverride=&Fort_Creatures_Vocalization_Tactical}
 }
 {
  id=SoundClass_7
  children [Attenuation_7]
  properties {soundClassOverride=&Fort_Creatures_Vocalization_Tactical}
 }
 {id=WavePlayer_0, properties {soundWaveAssetPtr=/Irwin/Sounds/GrandmaC/Waves/GrandmaC_Vox_Swim_Aggro_Breath_01}}
 {id=WavePlayer_1, properties {soundWaveAssetPtr=/Irwin/Sounds/GrandmaC/Waves/GrandmaC_Vox_Swim_Aggro_Breath_02}}
 {id=WavePlayer_14, properties {soundWaveAssetPtr=/Irwin/Sounds/GrandmaC/Waves/GrandmaC_Vox_Swim_Aggro_Breath_05}}
 {id=WavePlayer_15, properties {soundWaveAssetPtr=/Irwin/Sounds/GrandmaC/Waves/GrandmaC_Vox_Swim_Aggro_Breath_06}}
 {id=WavePlayer_16, properties {soundWaveAssetPtr=/Irwin/Sounds/GrandmaC/Waves/GrandmaC_Vox_Swim_Aggro_Breath_07}}
 {id=WavePlayer_17, properties {soundWaveAssetPtr=/Irwin/Sounds/GrandmaC/Waves/GrandmaC_Vox_Swim_Aggro_Breath_08}}
 {id=WavePlayer_18, properties {soundWaveAssetPtr=/Irwin/Sounds/GrandmaC/Waves/GrandmaC_Vox_Swim_Aggro_Breath_09}}
 {id=WavePlayer_19, properties {soundWaveAssetPtr=/Irwin/Sounds/GrandmaC/Waves/GrandmaC_Vox_Swim_Aggro_Breath_10}}
 {id=WavePlayer_2, properties {soundWaveAssetPtr=/Irwin/Sounds/GrandmaC/Waves/GrandmaC_Vox_Swim_Aggro_Breath_03}}
 {id=WavePlayer_20, properties {soundWaveAssetPtr=/Irwin/Sounds/GrandmaC/Waves/GrandmaC_Vox_Swim_Aggro_Breath_11}}
 {id=WavePlayer_21, properties {soundWaveAssetPtr=/Irwin/Sounds/GrandmaC/Waves/GrandmaC_Vox_Swim_Aggro_Breath_12}}
 {id=WavePlayer_22, properties {soundWaveAssetPtr=/Irwin/Sounds/GrandmaC/Waves/GrandmaC_Vox_Swim_Aggro_Breath_13}}
 {id=WavePlayer_23, properties {soundWaveAssetPtr=/Irwin/Sounds/GrandmaC/Waves/GrandmaC_Vox_Swim_Aggro_Breath_14}}
 {id=WavePlayer_24, properties {soundWaveAssetPtr=/Irwin/Sounds/GrandmaC/Waves/GrandmaC_Vox_Swim_Aggro_Breath_15}}
 {id=WavePlayer_25, properties {soundWaveAssetPtr=/Irwin/Sounds/GrandmaC/Waves/GrandmaC_Vox_Swim_Aggro_Breath_16}}
 {id=WavePlayer_26, properties {soundWaveAssetPtr=/Irwin/Sounds/GrandmaC/Waves/GrandmaC_Vox_Swim_Aggro_Breath_17}}
 {id=WavePlayer_27, properties {soundWaveAssetPtr=/Irwin/Sounds/GrandmaC/Waves/GrandmaC_Vox_Swim_Aggro_Breath_18}}
 {id=WavePlayer_28, properties {soundWaveAssetPtr=/Irwin/Sounds/GrandmaC/Waves/GrandmaC_Vox_Swim_Aggro_Breath_19}}
 {id=WavePlayer_29, properties {soundWaveAssetPtr=/Irwin/Sounds/GrandmaC/Waves/GrandmaC_Vox_Swim_Aggro_Breath_20}}
 {id=WavePlayer_3, properties {soundWaveAssetPtr=/Irwin/Sounds/GrandmaC/Waves/GrandmaC_Vox_Swim_Aggro_Breath_04}}
 {id=WavePlayer_30, properties {soundWaveAssetPtr=/Irwin/Sounds/GrandmaC/Waves/GrandmaC_Vox_Swim_Aggro_Breath_21}}
 {id=WavePlayer_31, properties {soundWaveAssetPtr=&GrandmaC_Vox_Swim_Aggro_Breath_01}}
 {id=WavePlayer_47, properties {soundWaveAssetPtr=&GrandmaC_Vox_Swim_Aggro_Breath_02}}
 {id=WavePlayer_48, properties {soundWaveAssetPtr=&GrandmaC_Vox_Swim_Aggro_Breath_03}}
 {id=WavePlayer_49, properties {soundWaveAssetPtr=&GrandmaC_Vox_Swim_Aggro_Breath_04}}
]
properties {
 maxDistance=10000
 volumeMultiplier=1.8
 concurrencySet [&Athena_Footstep_Local_Conc, &Athena_Footstep_Global_Conc]
}
