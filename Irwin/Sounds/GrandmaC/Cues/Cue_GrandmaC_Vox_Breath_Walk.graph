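# Sound cue for GrandmaC's walking breath vocals, varying with fatigue and with local, ridden-by-enemy and enemy state.
# IsLocal and IsRiddenByEnemy branches set sound class and volume, then share wolf-minion attenuation and random volume modulation.
# A quality-level node selects a larger or smaller FatigueLevel crossfade from normal breaths (aggro if IsEnemy) toward heavier breathing.
# Some random pools include empty slots, and concurrency rules limit overlap.
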SoundCue /Irwin/Sounds/GrandmaC/Cues/Cue_GrandmaC_Vox_Breath_Walk
root=Branch_4
nodes [
 {id=Attenuation_0, children [Modulator_3], properties {attenuationSettings=/Irwin/Sounds/GrandmaC/Settings/Att_WolfMinion_Vocalization}}
 {id=Attenuation_1, children [Modulator_3], properties {attenuationSettings=&Att_WolfMinion_Vocalization}}
 {id=Attenuation_2, children [Modulator_3], properties {attenuationSettings=&Att_WolfMinion_Vocalization}}
 {id=Attenuation_7, children [Modulator_3], properties {attenuationSettings=&Att_WolfMinion_Vocalization}}
 {id=Branch_0, children [Modulator_5,Modulator_4,SoundClass_2], properties {boolParameterName=IsRiddenByEnemy}}
 {id=Branch_1, children [Random_0,Random_2,Random_2], properties {boolParameterName=IsEnemy}}
 {id=Branch_2, children [Random_5,Random_13,Random_13], properties {boolParameterName=IsEnemy}}
 {id=Branch_4, children [SoundClass_7,Branch_0,Branch_0], properties {boolParameterName=IsLocal}}
 {id=Modulator_3, children [QualityLevel_0], properties {volumeMax=1, volumeMin=0.5}}
 {id=Modulator_4, children [SoundClass_1], properties {pitchMax=1, pitchMin=1, volumeMax=0.1, volumeMin=0.1}}
 {id=Modulator_5, children [SoundClass_0], properties {pitchMax=1, pitchMin=1, volumeMax=1, volumeMin=1}}
 {
  id=ParamCrossFade_1
  children [Branch_1,Random_6,Random_7]
  properties {
   paramName=FatigueLevel
   crossFadeInput [
    {fadeInDistanceEnd=0, fadeInDistanceStart=0, fadeOutDistanceEnd=0.85, fadeOutDistanceStart=0.82, volume=1.3}
    {fadeInDistanceEnd=0.85, fadeInDistanceStart=0.82, fadeOutDistanceEnd=0.96, fadeOutDistanceStart=0.95, volume=2.5}
    {fadeInDistanceEnd=0.96, fadeInDistanceStart=0.95, fadeOutDistanceEnd=1, fadeOutDistanceStart=1, volume=2.2}
   ]
  }
 }
 {
  id=ParamCrossFade_2
  children [Branch_2,Random_12,Random_8]
  properties {
   paramName=FatigueLevel
   crossFadeInput [
    {fadeInDistanceEnd=0, fadeInDistanceStart=0, fadeOutDistanceEnd=1, fadeOutDistanceStart=0.9, volume=1}
    {fadeInDistanceEnd=0.92, fadeInDistanceStart=0.88, fadeOutDistanceEnd=1, fadeOutDistanceStart=1, volume=1.5}
    {fadeInDistanceEnd=0, fadeInDistanceStart=0, fadeOutDistanceEnd=0, fadeOutDistanceStart=0, volume=1}
   ]
  }
 }
 {id=QualityLevel_0, children [ParamCrossFade_1, ParamCrossFade_1, ParamCrossFade_2, ParamCrossFade_2]}
 {
  id=Random_0
  children [
   WavePlayer_0
   WavePlayer_1
   WavePlayer_2
   WavePlayer_3
   WavePlayer_4
   WavePlayer_5
   WavePlayer_6
   WavePlayer_7
   WavePlayer_8
   WavePlayer_9
   WavePlayer_10
   WavePlayer_11
   WavePlayer_12
   WavePlayer_13
   WavePlayer_14
   WavePlayer_15
   WavePlayer_16
   WavePlayer_17
   WavePlayer_18
   WavePlayer_19
   WavePlayer_20
  ]
  properties {bShouldExcludeFromBranchCulling=true, weights [1,1,1,1,1,1,1,1,1,1,1,1,1,1,1,1,1,1,1,1,1]}
 }
 {
  id=Random_12
  children [WavePlayer_108, WavePlayer_109, WavePlayer_110, WavePlayer_111, null, null, null, null, null, null]
  properties {bShouldExcludeFromBranchCulling=true, weights [1,1,1,1,1,1,1,1,1,1]}
 }
 {
  id=Random_13
  children [WavePlayer_104, WavePlayer_105, WavePlayer_106, WavePlayer_107, null, null, null, null]
  properties {bShouldExcludeFromBranchCulling=true, weights [1,1,1,1,1,1,1,1]}
 }
 {
  id=Random_2
  children [
   WavePlayer_21
   WavePlayer_22
   WavePlayer_23
   WavePlayer_24
   WavePlayer_25
   WavePlayer_48
   WavePlayer_51
   WavePlayer_54
   WavePlayer_57
   WavePlayer_59
   WavePlayer_60
   WavePlayer_61
   WavePlayer_62
   WavePlayer_63
   WavePlayer_64
   WavePlayer_66
   WavePlayer_68
   WavePlayer_70
   WavePlayer_71
   WavePlayer_72
   WavePlayer_74
  ]
  properties {bShouldExcludeFromBranchCulling=true, weights [1,1,1,1,1,1,1,1,1,1,1,1,1,1,1,1,1,1,1,1,1]}
 }
 {
  id=Random_5
  children [WavePlayer_100, WavePlayer_101, WavePlayer_102, WavePlayer_103, null, null, null, null]
  properties {bShouldExcludeFromBranchCulling=true, weights [1,1,1,1,1,1,1,1]}
 }
 {
  id=Random_6
  children [
   WavePlayer_75
   WavePlayer_79
   WavePlayer_80
   WavePlayer_81
   WavePlayer_82
   WavePlayer_86
   WavePlayer_87
   WavePlayer_88
   null
   null
   null
   null
   null
   null
  ]
  properties {bShouldExcludeFromBranchCulling=true, weights [1,1,1,1,1,1,1,1,1,1,1,1,1,1]}
 }
 {
  id=Random_7
  children [WavePlayer_90, WavePlayer_91, WavePlayer_92, WavePlayer_93, WavePlayer_97, WavePlayer_98, WavePlayer_99, null, null, null]
  properties {bShouldExcludeFromBranchCulling=true, weights [1,1,1,1,1,1,1,1,1,1]}
 }
 {
  id=Random_8
  children [WavePlayer_112, WavePlayer_113, WavePlayer_114, WavePlayer_115, WavePlayer_116, WavePlayer_117, WavePlayer_118, null, null, null, null, null]
  properties {bShouldExcludeFromBranchCulling=true, weights [1,1,1,1,1,1,1,1,1,1,1,1]}
 }
 {
  id=SoundClass_0
  children [Attenuation_0]
  properties {soundClassOverride=&Fort_Player_Foley_3P_Enemy_Athena}
 }
 {
  id=SoundClass_1
  children [Attenuation_1]
  properties {soundClassOverride=&Fort_Creatures_Vocalization_Tactical}
 }
 {
  id=SoundClass_2
  children [Attenuation_2]
  properties {soundClassOverride=&Fort_Creatures_Vocalization_Tactical}
 }
 {
  id=SoundClass_7
  children [Attenuation_7]
  properties {soundClassOverride=&Fort_Creatures_Vocalization_Tactical}
 }
 {id=WavePlayer_0, properties {soundWaveAssetPtr=/Irwin/Sounds/GrandmaC/Waves/GrandmaC_Vox_Swim_Aggro_Breath_01}}
 {id=WavePlayer_1, properties {soundWaveAssetPtr=/Irwin/Sounds/GrandmaC/Waves/GrandmaC_Vox_Swim_Aggro_Breath_02}}
 {id=WavePlayer_10, properties {soundWaveAssetPtr=/Irwin/Sounds/GrandmaC/Waves/GrandmaC_Vox_Swim_Aggro_Breath_11}}
 {id=WavePlayer_100, properties {soundWaveAssetPtr=&GrandmaC_Vox_Swim_Aggro_Breath_01}}
 {id=WavePlayer_101, properties {soundWaveAssetPtr=&GrandmaC_Vox_Swim_Aggro_Breath_02}}
 {id=WavePlayer_102, properties {soundWaveAssetPtr=/Irwin/Sounds/GrandmaC/Waves/GrandmaC_Vox_Swim_Aggro_Breath_03}}
 {id=WavePlayer_103, properties {soundWaveAssetPtr=/Irwin/Sounds/GrandmaC/Waves/GrandmaC_Vox_Swim_Aggro_Breath_04}}
 {id=WavePlayer_104, properties {soundWaveAssetPtr=/Irwin/Sounds/GrandmaC/Waves/GrandmaC_Vox_Walk_Breath_01}}
 {id=WavePlayer_105, properties {soundWaveAssetPtr=/Irwin/Sounds/GrandmaC/Waves/GrandmaC_Vox_Walk_Breath_02}}
 {id=WavePlayer_106, properties {soundWaveAssetPtr=/Irwin/Sounds/GrandmaC/Waves/GrandmaC_Vox_Walk_Breath_03}}
 {id=WavePlayer_107, properties {soundWaveAssetPtr=/Irwin/Sounds/GrandmaC/Waves/GrandmaC_Vox_Walk_Breath_04}}
 {id=WavePlayer_108, properties {soundWaveAssetPtr=/Irwin/Sounds/GrandmaC/Waves/GrandmaC_Vocal_Heavy_Breathing_01}}
 {id=WavePlayer_109, properties {soundWaveAssetPtr=/Irwin/Sounds/GrandmaC/Waves/GrandmaC_Vocal_Heavy_Breathing_02}}
 {id=WavePlayer_11, properties {soundWaveAssetPtr=/Irwin/Sounds/GrandmaC/Waves/GrandmaC_Vox_Swim_Aggro_Breath_12}}
 {id=WavePlayer_110, properties {soundWaveAssetPtr=/Irwin/Sounds/GrandmaC/Waves/GrandmaC_Vocal_Heavy_Breathing_03}}
 {id=WavePlayer_111, properties {soundWaveAssetPtr=/Irwin/Sounds/GrandmaC/Waves/GrandmaC_Vocal_Heavy_Breathing_04}}
 {id=WavePlayer_112, properties {soundWaveAssetPtr=/Irwin/Sounds/GrandmaC/Waves/GrandmaC_Vocal_Sprint_Tiring_Whines_01}}
 {id=WavePlayer_113, properties {soundWaveAssetPtr=/Irwin/Sounds/GrandmaC/Waves/GrandmaC_Vocal_Sprint_Tiring_Whines_02}}
 {id=WavePlayer_114, properties {soundWaveAssetPtr=/Irwin/Sounds/GrandmaC/Waves/GrandmaC_Vocal_Sprint_Tiring_Whines_03}}
 {id=WavePlayer_115, properties {soundWaveAssetPtr=/Irwin/Sounds/GrandmaC/Waves/GrandmaC_Vocal_Sprint_Tiring_Whines_04}}
 {id=WavePlayer_116, properties {soundWaveAssetPtr=/Irwin/Sounds/GrandmaC/Waves/GrandmaC_Vox_Run_Aggro_Breath_01}}
 {id=WavePlayer_117, properties {soundWaveAssetPtr=/Irwin/Sounds/GrandmaC/Waves/GrandmaC_Vox_Run_Aggro_Breath_02}}
 {id=WavePlayer_118, properties {soundWaveAssetPtr=/Irwin/Sounds/GrandmaC/Waves/GrandmaC_Vox_Run_Aggro_Breath_03}}
 {id=WavePlayer_12, properties {soundWaveAssetPtr=/Irwin/Sounds/GrandmaC/Waves/GrandmaC_Vox_Swim_Aggro_Breath_13}}
 {id=WavePlayer_13, properties {soundWaveAssetPtr=/Irwin/Sounds/GrandmaC/Waves/GrandmaC_Vox_Swim_Aggro_Breath_14}}
 {id=WavePlayer_14, properties {soundWaveAssetPtr=/Irwin/Sounds/GrandmaC/Waves/GrandmaC_Vox_Swim_Aggro_Breath_15}}
 {id=WavePlayer_15, properties {soundWaveAssetPtr=/Irwin/Sounds/GrandmaC/Waves/GrandmaC_Vox_Swim_Aggro_Breath_16}}
 {id=WavePlayer_16, properties {soundWaveAssetPtr=/Irwin/Sounds/GrandmaC/Waves/GrandmaC_Vox_Swim_Aggro_Breath_17}}
 {id=WavePlayer_17, properties {soundWaveAssetPtr=/Irwin/Sounds/GrandmaC/Waves/GrandmaC_Vox_Swim_Aggro_Breath_18}}
 {id=WavePlayer_18, properties {soundWaveAssetPtr=/Irwin/Sounds/GrandmaC/Waves/GrandmaC_Vox_Swim_Aggro_Breath_19}}
 {id=WavePlayer_19, properties {soundWaveAssetPtr=/Irwin/Sounds/GrandmaC/Waves/GrandmaC_Vox_Swim_Aggro_Breath_20}}
 {id=WavePlayer_2, properties {soundWaveAssetPtr=&GrandmaC_Vox_Swim_Aggro_Breath_03}}
 {id=WavePlayer_20, properties {soundWaveAssetPtr=/Irwin/Sounds/GrandmaC/Waves/GrandmaC_Vox_Swim_Aggro_Breath_21}}
 {id=WavePlayer_21, properties {soundWaveAssetPtr=&GrandmaC_Vox_Walk_Breath_01}}
 {id=WavePlayer_22, properties {soundWaveAssetPtr=&GrandmaC_Vox_Walk_Breath_02}}
 {id=WavePlayer_23, properties {soundWaveAssetPtr=&GrandmaC_Vox_Walk_Breath_03}}
 {id=WavePlayer_24, properties {soundWaveAssetPtr=&GrandmaC_Vox_Walk_Breath_04}}
 {id=WavePlayer_25, properties {soundWaveAssetPtr=/Irwin/Sounds/GrandmaC/Waves/GrandmaC_Vox_Walk_Breath_05}}
 {id=WavePlayer_3, properties {soundWaveAssetPtr=&GrandmaC_Vox_Swim_Aggro_Breath_04}}
 {id=WavePlayer_4, properties {soundWaveAssetPtr=/Irwin/Sounds/GrandmaC/Waves/GrandmaC_Vox_Swim_Aggro_Breath_05}}
 {id=WavePlayer_48, properties {soundWaveAssetPtr=/Irwin/Sounds/GrandmaC/Waves/GrandmaC_Vox_Walk_Breath_06}}
 {id=WavePlayer_5, properties {soundWaveAssetPtr=/Irwin/Sounds/GrandmaC/Waves/GrandmaC_Vox_Swim_Aggro_Breath_06}}
 {id=WavePlayer_51, properties {soundWaveAssetPtr=/Irwin/Sounds/GrandmaC/Waves/GrandmaC_Vox_Walk_Breath_07}}
 {id=WavePlayer_54, properties {soundWaveAssetPtr=/Irwin/Sounds/GrandmaC/Waves/GrandmaC_Vox_Walk_Breath_08}}
 {id=WavePlayer_57, properties {soundWaveAssetPtr=/Irwin/Sounds/GrandmaC/Waves/GrandmaC_Vox_Walk_Breath_09}}
 {id=WavePlayer_59, properties {soundWaveAssetPtr=/Irwin/Sounds/GrandmaC/Waves/GrandmaC_Vox_Walk_Breath_10}}
 {id=WavePlayer_6, properties {soundWaveAssetPtr=/Irwin/Sounds/GrandmaC/Waves/GrandmaC_Vox_Swim_Aggro_Breath_07}}
 {id=WavePlayer_60, properties {soundWaveAssetPtr=/Irwin/Sounds/GrandmaC/Waves/GrandmaC_Vox_Walk_Breath_11}}
 {id=WavePlayer_61, properties {soundWaveAssetPtr=/Irwin/Sounds/GrandmaC/Waves/GrandmaC_Vox_Walk_Breath_12}}
 {id=WavePlayer_62, properties {soundWaveAssetPtr=/Irwin/Sounds/GrandmaC/Waves/GrandmaC_Vox_Walk_Breath_13}}
 {id=WavePlayer_63, properties {soundWaveAssetPtr=/Irwin/Sounds/GrandmaC/Waves/GrandmaC_Vox_Walk_Breath_14}}
 {id=WavePlayer_64, properties {soundWaveAssetPtr=/Irwin/Sounds/GrandmaC/Waves/GrandmaC_Vox_Walk_Breath_15}}
 {id=WavePlayer_66, properties {soundWaveAssetPtr=/Irwin/Sounds/GrandmaC/Waves/GrandmaC_Vox_Walk_Breath_16}}
 {id=WavePlayer_68, properties {soundWaveAssetPtr=/Irwin/Sounds/GrandmaC/Waves/GrandmaC_Vox_Walk_Breath_17}}
 {id=WavePlayer_7, properties {soundWaveAssetPtr=/Irwin/Sounds/GrandmaC/Waves/GrandmaC_Vox_Swim_Aggro_Breath_08}}
 {id=WavePlayer_70, properties {soundWaveAssetPtr=/Irwin/Sounds/GrandmaC/Waves/GrandmaC_Vox_Walk_Breath_18}}
 {id=WavePlayer_71, properties {soundWaveAssetPtr=/Irwin/Sounds/GrandmaC/Waves/GrandmaC_Vox_Walk_Breath_19}}
 {id=WavePlayer_72, properties {soundWaveAssetPtr=/Irwin/Sounds/GrandmaC/Waves/GrandmaC_Vox_Walk_Breath_20}}
 {id=WavePlayer_74, properties {soundWaveAssetPtr=/Irwin/Sounds/GrandmaC/Waves/GrandmaC_Vox_Walk_Breath_21}}
 {id=WavePlayer_75, properties {soundWaveAssetPtr=&GrandmaC_Vocal_Heavy_Breathing_01}}
 {id=WavePlayer_79, properties {soundWaveAssetPtr=&GrandmaC_Vocal_Heavy_Breathing_02}}
 {id=WavePlayer_8, properties {soundWaveAssetPtr=/Irwin/Sounds/GrandmaC/Waves/GrandmaC_Vox_Swim_Aggro_Breath_09}}
 {id=WavePlayer_80, properties {soundWaveAssetPtr=&GrandmaC_Vocal_Heavy_Breathing_03}}
 {id=WavePlayer_81, properties {soundWaveAssetPtr=&GrandmaC_Vocal_Heavy_Breathing_04}}
 {id=WavePlayer_82, properties {soundWaveAssetPtr=/Irwin/Sounds/GrandmaC/Waves/GrandmaC_Vocal_Heavy_Breathing_05}}
 {id=WavePlayer_86, properties {soundWaveAssetPtr=/Irwin/Sounds/GrandmaC/Waves/GrandmaC_Vocal_Heavy_Breathing_06}}
 {id=WavePlayer_87, properties {soundWaveAssetPtr=/Irwin/Sounds/GrandmaC/Waves/GrandmaC_Vocal_Heavy_Breathing_07}}
 {id=WavePlayer_88, properties {soundWaveAssetPtr=/Irwin/Sounds/GrandmaC/Waves/GrandmaC_Vocal_Heavy_Breathing_08}}
 {id=WavePlayer_9, properties {soundWaveAssetPtr=/Irwin/Sounds/GrandmaC/Waves/GrandmaC_Vox_Swim_Aggro_Breath_10}}
 {id=WavePlayer_90, properties {soundWaveAssetPtr=&GrandmaC_Vocal_Sprint_Tiring_Whines_01}}
 {id=WavePlayer_91, properties {soundWaveAssetPtr=&GrandmaC_Vocal_Sprint_Tiring_Whines_02}}
 {id=WavePlayer_92, properties {soundWaveAssetPtr=&GrandmaC_Vocal_Sprint_Tiring_Whines_03}}
 {id=WavePlayer_93, properties {soundWaveAssetPtr=&GrandmaC_Vocal_Sprint_Tiring_Whines_04}}
 {id=WavePlayer_97, properties {soundWaveAssetPtr=&GrandmaC_Vox_Run_Aggro_Breath_01}}
 {id=WavePlayer_98, properties {soundWaveAssetPtr=&GrandmaC_Vox_Run_Aggro_Breath_02}}
 {id=WavePlayer_99, properties {soundWaveAssetPtr=&GrandmaC_Vox_Run_Aggro_Breath_03}}
]
properties {
 maxDistance=10000
 pitchMultiplier=0.8
 volumeMultiplier=1
 concurrencySet [
  &Athena_Footstep_Local_Conc
  &Athena_Footstep_Global_Conc
  /Game/Sounds/Concurrency_Rules/LimitToOnePreventNewQuick
 ]
}
