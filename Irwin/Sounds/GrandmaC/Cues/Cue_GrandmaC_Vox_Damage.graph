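# Sound cue for GrandmaC damage vocalizations, played on the creature vocalization sound class with a boosted volume.
# A quality-level switch sends the first two levels to an equally weighted random pick of nine waves, excluded from branch culling.
# The last two levels use a five-wave subset passed through a default Modulator for variation.
# It borrows WolfMinion vocalization attenuation and uses Athena local and global footstep concurrency sets.
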SoundCue /Irwin/Sounds/GrandmaC/Cues/Cue_GrandmaC_Vox_Damage
root=QualityLevel_1
nodes [
 {id=Modulator_0, children [Random_3]}
 {id=QualityLevel_1, children [Random_0,Random_0,Modulator_0,Modulator_0]}
 {
  id=Random_0
  children [WavePlayer_0, WavePlayer_1, WavePlayer_2, WavePlayer_4, WavePlayer_5, WavePlayer_7, WavePlayer_8, WavePlayer_9, WavePlayer_10]
  properties {bShouldExcludeFromBranchCulling=true, weights [1,1,1,1,1,1,1,1,1]}
 }
 {id=Random_3, children [WavePlayer_11, WavePlayer_12, WavePlayer_19, WavePlayer_13, WavePlayer_14], properties {weights [1,1,1,1,1]}}
 {id=WavePlayer_0, properties {soundWaveAssetPtr=/Irwin/Sounds/GrandmaC/Waves/GrandmaC_Vox_Damage_01}}
 {id=WavePlayer_1, properties {soundWaveAssetPtr=/Irwin/Sounds/GrandmaC/Waves/GrandmaC_Vox_Damage_02}}
 {id=WavePlayer_10, properties {soundWaveAssetPtr=/Irwin/Sounds/GrandmaC/Waves/GrandmaC_Vox_Damage_11}}
 {id=WavePlayer_11, properties {soundWaveAssetPtr=&GrandmaC_Vox_Damage_01}}
 {id=WavePlayer_12, properties {soundWaveAssetPtr=&GrandmaC_Vox_Damage_02}}
 {id=WavePlayer_13, properties {soundWaveAssetPtr=/Irwin/Sounds/GrandmaC/Waves/GrandmaC_Vox_Damage_05}}
 {id=WavePlayer_14, properties {soundWaveAssetPtr=/Irwin/Sounds/GrandmaC/Waves/GrandmaC_Vox_Damage_06}}
 {id=WavePlayer_19, properties {soundWaveAssetPtr=/Irwin/Sounds/GrandmaC/Waves/GrandmaC_Vox_Damage_03}}
 {id=WavePlayer_2, properties {soundWaveAssetPtr=&GrandmaC_Vox_Damage_03}}
 {id=WavePlayer_4, properties {soundWaveAssetPtr=&GrandmaC_Vox_Damage_05}}
 {id=WavePlayer_5, properties {soundWaveAssetPtr=&GrandmaC_Vox_Damage_06}}
 {id=WavePlayer_7, properties {soundWaveAssetPtr=/Irwin/Sounds/GrandmaC/Waves/GrandmaC_Vox_Damage_08}}
 {id=WavePlayer_8, properties {soundWaveAssetPtr=/Irwin/Sounds/GrandmaC/Waves/GrandmaC_Vox_Damage_09}}
 {id=WavePlayer_9, properties {soundWaveAssetPtr=/Irwin/Sounds/GrandmaC/Waves/GrandmaC_Vox_Damage_10}}
]
properties {
 attenuationSettings=/Irwin/Sounds/GrandmaC/Settings/Att_WolfMinion_Vocalization
 maxDistance=10000
 soundClassObject=&Fort_Creatures_Vocalization_Tactical
 volumeMultiplier=2.4
 concurrencySet [&Athena_Footstep_Local_Conc, &Athena_Footstep_Global_Conc]
}
