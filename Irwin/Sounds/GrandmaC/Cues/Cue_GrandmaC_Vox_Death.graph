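# Sound cue for GrandmaC's death vocalization, playing one of seven screech waves.
# A fixed short delay feeds a default modulator, then an equally weighted random pick exempt from branch culling.
# It uses the tactical creature vocalization class, boosted volume and WolfMinion vocalization attenuation.
# Its concurrency sets are the Athena local and global footstep ones, despite being a voice cue.
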SoundCue /Irwin/Sounds/GrandmaC/Cues/Cue_GrandmaC_Vox_Death
root=Delay_0
nodes [
 {id=Delay_0, children [Modulator_3], properties {delayMax=0.2, delayMin=0.2}}
 {id=Modulator_3, children [Random_4]}
 {
  id=Random_4
  children [WavePlayer_23, WavePlayer_24, WavePlayer_25, WavePlayer_26, WavePlayer_27, WavePlayer_28, WavePlayer_29]
  properties {bShouldExcludeFromBranchCulling=true, weights [1,1,1,1,1,1,1]}
 }
 {id=WavePlayer_23, properties {soundWaveAssetPtr=/Irwin/Sounds/GrandmaC/Waves/GrandmaC_Vox_Death_Screech_07}}
 {id=WavePlayer_24, properties {soundWaveAssetPtr=/Irwin/Sounds/GrandmaC/Waves/GrandmaC_Vox_Death_Screech_06}}
 {id=WavePlayer_25, properties {soundWaveAssetPtr=/Irwin/Sounds/GrandmaC/Waves/GrandmaC_Vox_Death_Screech_05}}
 {id=WavePlayer_26, properties {soundWaveAssetPtr=/Irwin/Sounds/GrandmaC/Waves/GrandmaC_Vox_Death_Screech_04}}
 {id=WavePlayer_27, properties {soundWaveAssetPtr=/Irwin/Sounds/GrandmaC/Waves/GrandmaC_Vox_Death_Screech_03}}
 {id=WavePlayer_28, properties {soundWaveAssetPtr=/Irwin/Sounds/GrandmaC/Waves/GrandmaC_Vox_Death_Screech_02}}
 {id=WavePlayer_29, properties {soundWaveAssetPtr=/Irwin/Sounds/GrandmaC/Waves/GrandmaC_Vox_Death_Screech_01}}
]
properties {
 attenuationSettings=/Irwin/Sounds/GrandmaC/Settings/Att_WolfMinion_Vocalization
 maxDistance=10000
 soundClassObject=&Fort_Creatures_Vocalization_Tactical
 volumeMultiplier=2.8
 concurrencySet [&Athena_Footstep_Local_Conc, &Athena_Footstep_Global_Conc]
}
