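# Sound cue for GrandmaC howl vocalizations, using wolf-minion attenuation and the tactical creature vocalization sound class.
# A quality-level switch sends its first two slots to a random pick of seven howls pitched down by a fixed modulator.
# The last two slots randomly pick from only the first five howls, without the pitch shift.
# Concurrency is governed by Athena local and global footstep sets rather than vocal-specific ones.
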SoundCue /Irwin/Sounds/GrandmaC/Cues/Cue_GrandmaC_Vox_Howl
root=QualityLevel_0
nodes [
 {id=Modulator_0, children [Random_4], properties {pitchMax=0.9, pitchMin=0.9, volumeMax=1, volumeMin=1}}
 {id=QualityLevel_0, children [Modulator_0,Modulator_0,Random_5,Random_5]}
 {
  id=Random_4
  children [WavePlayer_15, WavePlayer_16, WavePlayer_17, WavePlayer_18, WavePlayer_19, WavePlayer_22, WavePlayer_23]
  properties {bShouldExcludeFromBranchCulling=true, weights [1,1,1,1,1,1,1]}
 }
 {id=Random_5, children [WavePlayer_24, WavePlayer_25, WavePlayer_26, WavePlayer_27, WavePlayer_28], properties {weights [1,1,1,1,1]}}
 {id=WavePlayer_15, properties {soundWaveAssetPtr=/Irwin/Sounds/GrandmaC/Waves/GrandmaC_Vox_Howl_01}}
 {id=WavePlayer_16, properties {soundWaveAssetPtr=/Irwin/Sounds/GrandmaC/Waves/GrandmaC_Vox_Howl_02}}
 {id=WavePlayer_17, properties {soundWaveAssetPtr=/Irwin/Sounds/GrandmaC/Waves/GrandmaC_Vox_Howl_03}}
 {id=WavePlayer_18, properties {soundWaveAssetPtr=/Irwin/Sounds/GrandmaC/Waves/GrandmaC_Vox_Howl_04}}
 {id=WavePlayer_19, properties {soundWaveAssetPtr=/Irwin/Sounds/GrandmaC/Waves/GrandmaC_Vox_Howl_05}}
 {id=WavePlayer_22, properties {soundWaveAssetPtr=/Irwin/Sounds/GrandmaC/Waves/GrandmaC_Vox_Howl_06}}
 {id=WavePlayer_23, properties {soundWaveAssetPtr=/Irwin/Sounds/GrandmaC/Waves/GrandmaC_Vox_Howl_07}}
 {id=WavePlayer_24, properties {soundWaveAssetPtr=&GrandmaC_Vox_Howl_01}}
 {id=WavePlayer_25, properties {soundWaveAssetPtr=&GrandmaC_Vox_Howl_02}}
 {id=WavePlayer_26, properties {soundWaveAssetPtr=&GrandmaC_Vox_Howl_03}}
 {id=WavePlayer_27, properties {soundWaveAssetPtr=&GrandmaC_Vox_Howl_04}}
 {id=WavePlayer_28, properties {soundWaveAssetPtr=&GrandmaC_Vox_Howl_05}}
]
properties {
 attenuationSettings=/Irwin/Sounds/GrandmaC/Settings/Att_WolfMinion_Vocalization
 maxDistance=10000
 soundClassObject=&Fort_Creatures_Vocalization_Tactical
 volumeMultiplier=2
 concurrencySet [&Athena_Footstep_Local_Conc, &Athena_Footstep_Global_Conc]
}
