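# Sound cue for GrandmaC idle vocalizations, with randomized waves and slight volume variation, limited by footstep and per-owner VO concurrency sets.
# IsLocal and IsRiddenByEnemy branches choose sound class, attenuation and volume: enemy-ridden non-local playback gets enemy foley class and wolf-minion attenuation, while explicitly non-enemy-ridden is heavily quieted.
# All paths share one modulator feeding a quality-level switch between the full wave pool and a six-wave subset.
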SoundCue /Irwin/Sounds/GrandmaC/Cues/Cue_GrandmaC_Vox_Idle
root=Branch_4
nodes [
 {id=Attenuation_0, children [Modulator_3], properties {attenuationSettings=/Irwin/Sounds/Settings/Predator_Vocalization_Att}}
 {id=Attenuation_5, children [Modulator_3], properties {attenuationSettings=/Irwin/Sounds/GrandmaC/Settings/Att_WolfMinion_Idle}}
 {id=Attenuation_6, children [Modulator_3], properties {attenuationSettings=&Predator_Vocalization_Att}}
 {id=Attenuation_7, children [Modulator_3], properties {attenuationSettings=&Predator_Vocalization_Att}}
 {id=Branch_3, children [Modulator_5,Modulator_4,SoundClass_0], properties {boolParameterName=IsRiddenByEnemy}}
 {id=Branch_4, children [SoundClass_7,Branch_3,Branch_3], properties {boolParameterName=IsLocal}}
 {id=Modulator_3, children [QualityLevel_0], properties {volumeMax=1, volumeMin=0.9}}
 {id=Modulator_4, children [SoundClass_6], properties {pitchMax=1, pitchMin=1, volumeMax=0.1, volumeMin=0.1}}
 {id=Modulator_5, children [SoundClass_5], properties {pitchMax=1, pitchMin=1, volumeMax=0.8, volumeMin=0.8}}
 {id=QualityLevel_0, children [Random_2,Random_2,Random_0,Random_0]}
 {
  id=Random_0
  children [WavePlayer_28, WavePlayer_29, WavePlayer_30, WavePlayer_31, WavePlayer_32, WavePlayer_33]
  properties {weights [1,1,1,1,1,1]}
 }
 {
  id=Random_2
  children [
   WavePlayer_0
   WavePlayer_1
   WavePlayer_2
   WavePlayer_3
   WavePlayer_4
   WavePlayer_5
   WavePlayer_6
   WavePlayer_7
   WavePlayer_8
   WavePlayer_18
   WavePlayer_19
   WavePlayer_20
   WavePlayer_21
   WavePlayer_22
   WavePlayer_23
   WavePlayer_24
   WavePlayer_25
   WavePlayer_26
   WavePlayer_27
  ]
  properties {bShouldExcludeFromBranchCulling=true, weights [1,1,1,1,1,1,1,1,1,1,1,1,1,1,1,1,1,1,1]}
 }
 {
  id=SoundClass_0
  children [Attenuation_0]
  properties {soundClassOverride=&Fort_Creatures_Vocalization_Tactical}
 }
 {
  id=SoundClass_5
  children [Attenuation_5]
  properties {soundClassOverride=&Fort_Player_Foley_3P_Enemy_Athena}
 }
 {
  id=SoundClass_6
  children [Attenuation_6]
  properties {soundClassOverride=&Fort_Creatures_Vocalization_Tactical}
 }
 {
  id=SoundClass_7
  children [Attenuation_7]
  properties {soundClassOverride=&Fort_Creatures_Vocalization_Tactical}
 }
 {id=WavePlayer_0, properties {soundWaveAssetPtr=/Irwin/Sounds/GrandmaC/Waves/GrandmaC_Vox_Idle_01}}
 {id=WavePlayer_1, properties {soundWaveAssetPtr=/Irwin/Sounds/GrandmaC/Waves/GrandmaC_Vox_Idle_02}}
 {id=WavePlayer_18, properties {soundWaveAssetPtr=/Irwin/Sounds/GrandmaC/Waves/GrandmaC_Vox_Idle_10}}
 {id=WavePlayer_19, properties {soundWaveAssetPtr=/Irwin/Sounds/GrandmaC/Waves/GrandmaC_Vox_Idle_11}}
 {id=WavePlayer_2, properties {soundWaveAssetPtr=/Irwin/Sounds/GrandmaC/Waves/GrandmaC_Vox_Idle_03}}
 {id=WavePlayer_20, properties {soundWaveAssetPtr=/Irwin/Sounds/GrandmaC/Waves/GrandmaC_Vox_Idle_12}}
 {id=WavePlayer_21, properties {soundWaveAssetPtr=/Irwin/Sounds/GrandmaC/Waves/GrandmaC_Vox_Idle_13}}
 {id=WavePlayer_22, properties {soundWaveAssetPtr=/Irwin/Sounds/GrandmaC/Waves/GrandmaC_Vox_Idle_14}}
 {id=WavePlayer_23, properties {soundWaveAssetPtr=/Irwin/Sounds/GrandmaC/Waves/GrandmaC_Vox_Idle_15}}
 {id=WavePlayer_24, properties {soundWaveAssetPtr=/Irwin/Sounds/GrandmaC/Waves/GrandmaC_Vox_Idle_16}}
 {id=WavePlayer_25, properties {soundWaveAssetPtr=/Irwin/Sounds/GrandmaC/Waves/GrandmaC_Vox_Idle_17}}
 {id=WavePlayer_26, properties {soundWaveAssetPtr=/Irwin/Sounds/GrandmaC/Waves/GrandmaC_Vox_Idle_18}}
 {id=WavePlayer_27, properties {soundWaveAssetPtr=/Irwin/Sounds/GrandmaC/Waves/GrandmaC_Vox_Idle_19}}
 {id=WavePlayer_28, properties {soundWaveAssetPtr=&GrandmaC_Vox_Idle_01}}
 {id=WavePlayer_29, properties {soundWaveAssetPtr=&GrandmaC_Vox_Idle_02}}
 {id=WavePlayer_3, properties {soundWaveAssetPtr=/Irwin/Sounds/GrandmaC/Waves/GrandmaC_Vox_Idle_04}}
 {id=WavePlayer_30, properties {soundWaveAssetPtr=&GrandmaC_Vox_Idle_03}}
 {id=WavePlayer_31, properties {soundWaveAssetPtr=&GrandmaC_Vox_Idle_04}}
 {id=WavePlayer_32, properties {soundWaveAssetPtr=/Irwin/Sounds/GrandmaC/Waves/GrandmaC_Vox_Idle_05}}
 {id=WavePlayer_33, properties {soundWaveAssetPtr=/Irwin/Sounds/GrandmaC/Waves/GrandmaC_Vox_Idle_06}}
 {id=WavePlayer_4, properties {soundWaveAssetPtr=&GrandmaC_Vox_Idle_05}}
 {id=WavePlayer_5, properties {soundWaveAssetPtr=&GrandmaC_Vox_Idle_06}}
 {id=WavePlayer_6, properties {soundWaveAssetPtr=/Irwin/Sounds/GrandmaC/Waves/GrandmaC_Vox_Idle_07}}
 {id=WavePlayer_7, properties {soundWaveAssetPtr=/Irwin/Sounds/GrandmaC/Waves/GrandmaC_Vox_Idle_08}}
 {id=WavePlayer_8, properties {soundWaveAssetPtr=/Irwin/Sounds/GrandmaC/Waves/GrandmaC_Vox_Idle_09}}
]
properties {
 maxDistance=9000
 volumeMultiplier=1.9
 concurrencySet [
  &Athena_Footstep_Local_Conc
  &Athena_Footstep_Global_Conc
  /Irwin/Sounds/Settings/Conc/SetOneLimitToOwner_VO_Conc
 ]
}
