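# Sound cue for GrandmaC jump vocalizations, picking one of four equally weighted waves with randomized modulation.
# IsLocal plays at full level; otherwise IsRiddenByEnemy keeps full level when true and drops to one-tenth volume when false or unset.
# All paths share the tactical creature vocalization class and WolfMinion attenuation.
# Concurrency reuses local and global footstep rules plus a limit-to-one rule.
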SoundCue /Irwin/Sounds/GrandmaC/Cues/Cue_GrandmaC_Vox_Jump
root=Branch_4
nodes [
 {id=Attenuation_5, children [Modulator_3], properties {attenuationSettings=/Irwin/Sounds/GrandmaC/Settings/Att_WolfMinion_Vocalization}}
 {id=Attenuation_6, children [Modulator_3], properties {attenuationSettings=&Att_WolfMinion_Vocalization}}
 {id=Attenuation_7, children [Modulator_3], properties {attenuationSettings=&Att_WolfMinion_Vocalization}}
 {id=Branch_3, children [Modulator_5,Modulator_4,Modulator_4], properties {boolParameterName=IsRiddenByEnemy}}
 {id=Branch_4, children [SoundClass_7,Branch_3,Branch_3], properties {boolParameterName=IsLocal}}
 {id=Modulator_3, children [Random_0], properties {volumeMax=1, volumeMin=0.5}}
 {id=Modulator_4, children [SoundClass_6], properties {pitchMax=1, pitchMin=1, volumeMax=0.1, volumeMin=0.1}}
 {id=Modulator_5, children [SoundClass_5], properties {pitchMax=1, pitchMin=1, volumeMax=1, volumeMin=1}}
 {
  id=Random_0
  children [WavePlayer_0,WavePlayer_5,WavePlayer_6,WavePlayer_7]
  properties {bShouldExcludeFromBranchCulling=true, weights [1,1,1,1]}
 }
 {
  id=SoundClass_5
  children [Attenuation_5]
  properties {soundClassOverride=&Fort_Creatures_Vocalization_Tactical}
 }
 {
  id=SoundClass_6
  children [Attenuation_6]
  properties {soundClassOverride=&Fort_Creatures_Vocalization_Tactical}
 }
 {
  id=SoundClass_7
  children [Attenuation_7]
  properties {soundClassOverride=&Fort_Creatures_Vocalization_Tactical}
 }
 {id=WavePlayer_0, properties {soundWaveAssetPtr=/Irwin/Sounds/GrandmaC/Waves/GrandmaC_Vox_Jump_01}}
 {id=WavePlayer_5, properties {soundWaveAssetPtr=/Irwin/Sounds/GrandmaC/Waves/GrandmaC_Vox_Jump_02}}
 {id=WavePlayer_6, properties {soundWaveAssetPtr=/Irwin/Sounds/GrandmaC/Waves/GrandmaC_Vox_Jump_03}}
 {id=WavePlayer_7, properties {soundWaveAssetPtr=/Irwin/Sounds/GrandmaC/Waves/GrandmaC_Vox_Jump_04}}
]
properties {
 maxDistance=10000
 volumeMultiplier=1.25
 concurrencySet [
  &Athena_Footstep_Local_Conc
  &Athena_Footstep_Global_Conc
  /Game/Sounds/Concurrency_Rules/LimitToOnePreventNew
 ]
}
