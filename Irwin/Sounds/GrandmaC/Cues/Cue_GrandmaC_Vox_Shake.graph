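# Sound cue for GrandmaC's shake vocalization, routed through the creature vocalization sound class.
# A Random node picks one of three equally weighted shake waves, and a Modulator at the root adds slight pitch variation.
# It uses the shared Predator vocalization attenuation and, despite being a vocal sound, the local and global Athena footstep concurrency groups.
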SoundCue /Irwin/Sounds/GrandmaC/Cues/Cue_GrandmaC_Vox_Shake
root=Modulator_0
nodes [
 {id=Modulator_0, children [Random_1], properties {pitchMax=1.02, pitchMin=0.98}}
 {id=Random_1, children [WavePlayer_0,WavePlayer_1,WavePlayer_2], properties {bShouldExcludeFromBranchCulling=true, weights [1,1,1]}}
 {id=WavePlayer_0, properties {soundWaveAssetPtr=/Irwin/Sounds/GrandmaC/Waves/GrandmaC_Vox_Shake_03}}
 {id=WavePlayer_1, properties {soundWaveAssetPtr=/Irwin/Sounds/GrandmaC/Waves/GrandmaC_Vox_Shake_02}}
 {id=WavePlayer_2, properties {soundWaveAssetPtr=/Irwin/Sounds/GrandmaC/Waves/GrandmaC_Vox_Shake_01}}
]
properties {
 attenuationSettings=/Irwin/Sounds/Settings/Predator_Vocalization_Att
 maxDistance=8300
 soundClassObject=/Game/Sounds/FortSoundClassesAndModes/Fort_Creatures_Vocalization_Normal
 volumeMultiplier=1.2
 concurrencySet [&Athena_Footstep_Local_Conc, &Athena_Footstep_Global_Conc]
}
